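# Sound cue for impacts on an ice hatch prop, filed under exterior demolition sounds, reusing IceMaiden pickaxe impact waves.
# A random node picks one of four equally weighted waves, feeding a modulator that slightly varies pitch.
# Close demolition attenuation, an impact sound class and a four-max concurrency set govern range, mixing and simultaneous playback.
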SoundCue /Figment_S10/Sounds/Fort_Impact_Sounds/Demolition/Exterior_Props/Demo_Hatch_IceProp_Impact_Cue
root=Modulator_2
nodes [
 {id=Modulator_2, children [Random_0], properties {pitchMax=1.07, pitchMin=0.92}}
 {id=Random_0, children [WavePlayer_1,WavePlayer_2,WavePlayer_3,WavePlayer_4], properties {weights [1,1,1,1]}}
 {id=WavePlayer_1, properties {soundWaveAssetPtr=/Figment_S10/Athena/Sounds/Weapons/PickAxes/IceMaiden/PA_IceMaiden_Impact_01}}
 {id=WavePlayer_2, properties {soundWaveAssetPtr=/Figment_S10/Athena/Sounds/Weapons/PickAxes/IceMaiden/PA_IceMaiden_Impact_02}}
 {id=WavePlayer_3, properties {soundWaveAssetPtr=/Figment_S10/Athena/Sounds/Weapons/PickAxes/IceMaiden/PA_IceMaiden_Impact_03}}
 {id=WavePlayer_4, properties {soundWaveAssetPtr=/Figment_S10/Athena/Sounds/Weapons/PickAxes/IceMaiden/PA_IceMaiden_Impact_04}}
]
properties {
 attenuationSettings=/Figment_S10/Sounds/Attenuation/Fort_Demolition_Close_Att
 maxDistance=2000
 soundClassObject=/Game/Sounds/FortSoundClassesAndModes/Fort_ImpactSounds_DuckDefault_PSM
 volumeMultiplier=0.8
 concurrencyOverrides {maxCount=4}
 concurrencySet [&CONC_Object_Impacts_FourMax]
}
