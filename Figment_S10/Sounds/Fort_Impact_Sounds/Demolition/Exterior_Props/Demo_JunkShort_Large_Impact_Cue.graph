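# Sound cue for short, large junk impact sounds in the demolition set for exterior props.
# A random node picks one of three equally weighted wave variations, and a modulator adds slight pitch variation to reduce repetition.
# Playback uses close-range demolition attenuation, a ducking impact sound class and a shared object-impact concurrency set that, by its name, caps simultaneous voices at four.
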SoundCue /Figment_S10/Sounds/Fort_Impact_Sounds/Demolition/Exterior_Props/Demo_JunkShort_Large_Impact_Cue
root=Modulator_2
nodes [
 {id=Modulator_2, children [Random_15], properties {pitchMax=1.07, pitchMin=0.92}}
 {id=Random_15, children [WavePlayer_91, WavePlayer_92, WavePlayer_94], properties {weights [1,1,1]}}
 {id=WavePlayer_91, properties {soundWaveAssetPtr=/Figment_S10/Sounds/Fort_Impact_Sounds/Demolition/Junk_ImpactShort_Large01}}
 {id=WavePlayer_92, properties {soundWaveAssetPtr=/Figment_S10/Sounds/Fort_Impact_Sounds/Demolition/Junk_ImpactShort_Large02}}
 {id=WavePlayer_94, properties {soundWaveAssetPtr=/Figment_S10/Sounds/Fort_Impact_Sounds/Demolition/Junk_ImpactShort_Large04}}
]
properties {
 attenuationSettings=/Figment_S10/Sounds/Attenuation/Fort_Demolition_Close_Att
 maxDistance=2000
 soundClassObject=/Game/Sounds/FortSoundClassesAndModes/Fort_ImpactSounds_DuckDefault_PSM
 volumeMultiplier=1
 concurrencyOverrides {maxCount=4}
 concurrencySet [&CONC_Object_Impacts_FourMax]
}
