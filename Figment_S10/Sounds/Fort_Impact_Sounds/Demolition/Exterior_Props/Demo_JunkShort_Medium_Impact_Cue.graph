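# Sound cue for a short, medium junk impact among the exterior prop demolition sounds.
# A random node picks one of three equally weighted wave variants, and a modulator varies its pitch slightly.
# It uses a close demolition attenuation, an impact sound class and a shared object-impact concurrency set that, by its name, caps simultaneous plays at four.
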SoundCue /Figment_S10/Sounds/Fort_Impact_Sounds/Demolition/Exterior_Props/Demo_JunkShort_Medium_Impact_Cue
root=Modulator_2
nodes [
 {id=Modulator_2, children [Random_14], properties {pitchMax=1.07, pitchMin=0.92}}
 {id=Random_14, children [WavePlayer_85, WavePlayer_86, WavePlayer_87], properties {weights [1,1,1]}}
 {id=WavePlayer_85, properties {soundWaveAssetPtr=/Figment_S10/Sounds/Fort_Impact_Sounds/Demolition/Junk_ImpactShort_Medium01}}
 {id=WavePlayer_86, properties {soundWaveAssetPtr=/Figment_S10/Sounds/Fort_Impact_Sounds/Demolition/Junk_ImpactShort_Medium02}}
 {id=WavePlayer_87, properties {soundWaveAssetPtr=/Figment_S10/Sounds/Fort_Impact_Sounds/Demolition/Junk_ImpactShort_Medium04}}
]
properties {
 attenuationSettings=/Figment_S10/Sounds/Attenuation/Fort_Demolition_Close_Att
 maxDistance=2000
 soundClassObject=/Game/Sounds/FortSoundClassesAndModes/Fort_ImpactSounds_DuckDefault_PSM
 volumeMultiplier=1
 concurrencyOverrides {maxCount=4}
 concurrencySet [&CONC_Object_Impacts_FourMax]
}
