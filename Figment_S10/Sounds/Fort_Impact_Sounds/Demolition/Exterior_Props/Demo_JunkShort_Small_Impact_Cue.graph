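# Sound cue for short, small junk impacts, filed under demolition sounds for exterior props.
# A Random node picks among three junk impact wave variants with equal weight, and a Modulator slightly varies its pitch on each play.
# It uses close demolition attenuation, a ducking impact sound class, and an object-impact concurrency set named for a four-voice limit.
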SoundCue /Figment_S10/Sounds/Fort_Impact_Sounds/Demolition/Exterior_Props/Demo_JunkShort_Small_Impact_Cue
root=Modulator_2
nodes [
 {id=Modulator_2, children [Random_13], properties {pitchMax=1.07, pitchMin=0.92}}
 {id=Random_13, children [WavePlayer_80, WavePlayer_81, WavePlayer_82], properties {weights [1,1,1]}}
 {id=WavePlayer_80, properties {soundWaveAssetPtr=/Figment_S10/Sounds/Fort_Impact_Sounds/Demolition/Junk_ImpactShort_Small01}}
 {id=WavePlayer_81, properties {soundWaveAssetPtr=/Figment_S10/Sounds/Fort_Impact_Sounds/Demolition/Junk_ImpactShort_Small03}}
 {id=WavePlayer_82, properties {soundWaveAssetPtr=/Figment_S10/Sounds/Fort_Impact_Sounds/Demolition/Junk_ImpactShort_Small04}}
]
properties {
 attenuationSettings=/Figment_S10/Sounds/Attenuation/Fort_Demolition_Close_Att
 maxDistance=2000
 soundClassObject=/Game/Sounds/FortSoundClassesAndModes/Fort_ImpactSounds_DuckDefault_PSM
 volumeMultiplier=1
 concurrencyOverrides {maxCount=4}
 concurrencySet [&CONC_Object_Impacts_FourMax]
}
