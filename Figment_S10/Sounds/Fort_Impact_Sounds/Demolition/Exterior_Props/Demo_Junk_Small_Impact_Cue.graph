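# Sound cue for small junk impacts in the demolition exterior prop impact set.
# A random node picks one of three equally weighted small junk impact waves and feeds a modulator that slightly varies pitch on each play.
# Playback uses close demolition attenuation, an impact sound class and an object-impact concurrency set limited to four voices.
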SoundCue /Figment_S10/Sounds/Fort_Impact_Sounds/Demolition/Exterior_Props/Demo_Junk_Small_Impact_Cue
root=Modulator_2
nodes [
 {id=Modulator_2, children [Random_16], properties {pitchMax=1.07, pitchMin=0.92}}
 {id=Random_16, children [WavePlayer_95, WavePlayer_96, WavePlayer_97], properties {weights [1,1,1]}}
 {id=WavePlayer_95, properties {soundWaveAssetPtr=/Figment_S10/Sounds/Fort_Impact_Sounds/Demolition/Junk_Impact_Small05}}
 {id=WavePlayer_96, properties {soundWaveAssetPtr=/Figment_S10/Sounds/Fort_Impact_Sounds/Demolition/Junk_Impact_Small01}}
 {id=WavePlayer_97, properties {soundWaveAssetPtr=/Figment_S10/Sounds/Fort_Impact_Sounds/Demolition/Junk_Impact_Small02}}
]
properties {
 attenuationSettings=/Figment_S10/Sounds/Attenuation/Fort_Demolition_Close_Att
 maxDistance=2000
 soundClassObject=/Game/Sounds/FortSoundClassesAndModes/Fort_ImpactSounds_DuckDefault_PSM
 volumeMultiplier=1
 concurrencyOverrides {maxCount=4}
 concurrencySet [&CONC_Object_Impacts_FourMax]
}
