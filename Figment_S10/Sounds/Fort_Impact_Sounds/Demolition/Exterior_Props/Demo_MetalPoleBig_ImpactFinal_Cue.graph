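# Demolition sound cue for the final impact on a big metal pole exterior prop.
# A quality-level switch picks between two three-layer mixes and two branches that play a single pre-optimized impact wave.
# The mixes layer randomized pole death variants (three or two), a small metal death and a higher-pitched hollow tank impact, with pitch modulation for variety.
# Close demolition attenuation, an impacts sound class and four-max object-impact concurrency govern playback.
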SoundCue /Figment_S10/Sounds/Fort_Impact_Sounds/Demolition/Exterior_Props/Demo_MetalPoleBig_ImpactFinal_Cue
root=QualityLevel_0
nodes [
 {id=Mixer_0, children [Modulator_0,Modulator_1,Modulator_3], properties {inputVolume [1,1,1]}}
 {id=Mixer_1, children [Modulator_2,Modulator_7,Modulator_8], properties {inputVolume [1,1,1]}}
 {id=Modulator_0, children [Random_0], properties {pitchMax=1.07, pitchMin=0.92}}
 {id=Modulator_1, children [WavePlayer_3]}
 {id=Modulator_2, children [Random_12], properties {pitchMax=1.07, pitchMin=0.92}}
 {id=Modulator_3, children [WavePlayer_4], properties {pitchMax=1.25, pitchMin=1.2}}
 {id=Modulator_4, children [WavePlayer_5], properties {pitchMax=1.07, pitchMin=0.92}}
 {id=Modulator_5, children [WavePlayer_6], properties {pitchMax=1.07, pitchMin=0.92}}
 {id=Modulator_7, children [WavePlayer_72]}
 {id=Modulator_8, children [WavePlayer_76], properties {pitchMax=1.25, pitchMin=1.2}}
 {id=QualityLevel_0, children [Mixer_1,Mixer_0,Modulator_4,Modulator_5]}
 {id=Random_0, children [WavePlayer_2,WavePlayer_0], properties {weights [1,1]}}
 {id=Random_12, children [WavePlayer_43, WavePlayer_42, WavePlayer_16], properties {weights [1,1,1]}}
 {id=WavePlayer_0, properties {soundWaveAssetPtr=/Figment_S10/Sounds/Fort_Impact_Sounds/Demolition/Metal_Pole_Death06}}
 {id=WavePlayer_16, properties {soundWaveAssetPtr=/Figment_S10/Sounds/Fort_Impact_Sounds/Demolition/Metal_Pole_Death02}}
 {id=WavePlayer_2, properties {soundWaveAssetPtr=/Figment_S10/Sounds/Fort_Impact_Sounds/Demolition/Metal_Pole_Death03}}
 {id=WavePlayer_3, properties {soundWaveAssetPtr=/Figment_S10/Sounds/Fort_Impact_Sounds/Demolition/Metal_Small_Death03}}
 {id=WavePlayer_4, properties {soundWaveAssetPtr=/Figment_S10/Sounds/Fort_Impact_Sounds/Demolition/Metal_HollowTank_Impact06}}
 {id=WavePlayer_42, properties {soundWaveAssetPtr=&Metal_Pole_Death06}}
 {id=WavePlayer_43, properties {soundWaveAssetPtr=&Metal_Pole_Death03}}
 {id=WavePlayer_5, properties {soundWaveAssetPtr=/Figment_S10/Sounds/Fort_Impact_Sounds/Demolition/Metal_PoleBig_ImpactFinal_opt}}
 {id=WavePlayer_6, properties {soundWaveAssetPtr=&Metal_PoleBig_ImpactFinal_opt}}
 {id=WavePlayer_72, properties {soundWaveAssetPtr=&Metal_Small_Death03}}
 {id=WavePlayer_76, properties {soundWaveAssetPtr=&Metal_HollowTank_Impact06}}
]
properties {
 attenuationSettings=/Figment_S10/Sounds/Attenuation/Fort_Demolition_Close_Att
 maxDistance=2000
 soundClassObject=/Game/Sounds/FortSoundClassesAndModes/Fort_ImpactSounds_DuckDefault_PSM
 volumeMultiplier=0.6
 concurrencyOverrides {resolutionRule=PreventNew}
 concurrencySet [&CONC_Object_Impacts_FourMax]
}
