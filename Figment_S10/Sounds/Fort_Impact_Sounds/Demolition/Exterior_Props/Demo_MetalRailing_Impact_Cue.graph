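# Sound cue for metal railing impacts in exterior-prop demolition, played at half volume with close-range demolition attenuation.
# A quality-level root selects among four branches: random picks from three waves or two waves, or one of two identical single-wave branches.
# Each branch adds slight random pitch modulation for variation.
# It uses a ducking impact sound class and an object-impact concurrency set to limit overlapping plays.
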SoundCue /Figment_S10/Sounds/Fort_Impact_Sounds/Demolition/Exterior_Props/Demo_MetalRailing_Impact_Cue
root=QualityLevel_0
nodes [
 {id=Modulator_0, children [Random_0], properties {pitchMax=1.07, pitchMin=0.92}}
 {id=Modulator_1, children [WavePlayer_3], properties {pitchMax=1.07, pitchMin=0.92}}
 {id=Modulator_10, children [Random_17], properties {pitchMax=1.07, pitchMin=0.92}}
 {id=Modulator_2, children [WavePlayer_4], properties {pitchMax=1.07, pitchMin=0.92}}
 {id=QualityLevel_0, children [Modulator_10,Modulator_0,Modulator_1,Modulator_2]}
 {id=Random_0, children [WavePlayer_0,WavePlayer_1], properties {weights [1,1]}}
 {id=Random_17, children [WavePlayer_60, WavePlayer_61, WavePlayer_62], properties {weights [1,1,1]}}
 {id=WavePlayer_0, properties {soundWaveAssetPtr=/Figment_S10/Sounds/Fort_Impact_Sounds/Demolition/Metal_Railing_Impact04}}
 {id=WavePlayer_1, properties {soundWaveAssetPtr=/Figment_S10/Sounds/Fort_Impact_Sounds/Demolition/Metal_Railing_Impact01}}
 {id=WavePlayer_3, properties {soundWaveAssetPtr=&Metal_Railing_Impact04}}
 {id=WavePlayer_4, properties {soundWaveAssetPtr=&Metal_Railing_Impact04}}
 {id=WavePlayer_60, properties {soundWaveAssetPtr=&Metal_Railing_Impact04}}
 {id=WavePlayer_61, properties {soundWaveAssetPtr=&Metal_Railing_Impact01}}
 {id=WavePlayer_62, properties {soundWaveAssetPtr=/Figment_S10/Sounds/Fort_Impact_Sounds/Demolition/Metal_Railing_Impact06}}
]
properties {
 attenuationSettings=/Figment_S10/Sounds/Attenuation/Fort_Demolition_Close_Att
 maxDistance=2000
 soundClassObject=/Game/Sounds/FortSoundClassesAndModes/Fort_ImpactSounds_DuckDefault_PSM
 volumeMultiplier=0.5
 concurrencyOverrides {maxCount=3, resolutionRule=StopOldest}
 concurrencySet [&CONC_Object_Impacts_FourMax]
}
