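# Sound cue for exterior metal sign demolition impacts, with close-range demolition attenuation and an impact sound class.
# A quality-level root sends its first two slots to a random pick of three waves and its last two to a pick of two.
# Each branch passes through a modulator for slight pitch variation.
# A shared object-impact concurrency set is assigned to limit overlapping plays.
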SoundCue /Figment_S10/Sounds/Fort_Impact_Sounds/Demolition/Exterior_Props/Demo_MetalSign_Impact_Cue
root=QualityLevel_0
nodes [
 {id=Modulator_0, children [Random_0], properties {pitchMax=1.07, pitchMin=0.92}}
 {id=Modulator_2, children [Random_7], properties {pitchMax=1.07, pitchMin=0.92}}
 {id=QualityLevel_0, children [Modulator_2,Modulator_2,Modulator_0,Modulator_0]}
 {id=Random_0, children [WavePlayer_0,WavePlayer_1], properties {weights [1,1]}}
 {id=Random_7, children [WavePlayer_25, WavePlayer_26, WavePlayer_27], properties {weights [1,1,1]}}
 {id=WavePlayer_0, properties {soundWaveAssetPtr=/Figment_S10/Sounds/Fort_Impact_Sounds/Demolition/Metal_Sign_Impact04}}
 {id=WavePlayer_1, properties {soundWaveAssetPtr=/Figment_S10/Sounds/Fort_Impact_Sounds/Demolition/Metal_Sign_Impact02}}
 {id=WavePlayer_25, properties {soundWaveAssetPtr=&Metal_Sign_Impact04}}
 {id=WavePlayer_26, properties {soundWaveAssetPtr=&Metal_Sign_Impact02}}
 {id=WavePlayer_27, properties {soundWaveAssetPtr=/Figment_S10/Sounds/Fort_Impact_Sounds/Demolition/Metal_Sign_Impact01}}
]
properties {
 attenuationSettings=/Figment_S10/Sounds/Attenuation/Fort_Demolition_Close_Att
 maxDistance=2000
 soundClassObject=/Game/Sounds/FortSoundClassesAndModes/Fort_ImpactSounds_DuckDefault_PSM
 volumeMultiplier=0.425
 concurrencyOverrides {maxCount=3, resolutionRule=StopOldest}
 concurrencySet [&CONC_Object_Impacts_FourMax]
}
