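# Sound cue for small metal impacts on demolished exterior props, using close demolition attenuation, an impact sound class and a concurrency set.
# A quality-level switch selects between two layered mixes and two single-wave fallbacks that share one optimized sample.
# Each mix blends small-metal impacts with quieter, pitched-up pole hits and faint wobbles, randomizing sample and pitch per layer.
# The top tier draws from three variants per layer, the second from two.
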SoundCue /Figment_S10/Sounds/Fort_Impact_Sounds/Demolition/Exterior_Props/Demo_MetalSmall_Impact_Cue
root=QualityLevel_0
nodes [
 {id=Mixer_0, children [Modulator_0,Modulator_2,Modulator_1], properties {inputVolume [1,0.75,0.25]}}
 {id=Mixer_3, children [Modulator_5,Modulator_7,Modulator_6], properties {inputVolume [1,0.75,0.25]}}
 {id=Modulator_0, children [Random_0], properties {pitchMax=1.07, pitchMin=0.92}}
 {id=Modulator_1, children [Random_1], properties {pitchMax=1.5, pitchMin=1.2}}
 {id=Modulator_2, children [Random_2], properties {pitchMax=1.62, pitchMin=1.5}}
 {id=Modulator_3, children [WavePlayer_9], properties {pitchMax=1.07, pitchMin=0.92}}
 {id=Modulator_4, children [WavePlayer_10], properties {pitchMax=1.07, pitchMin=0.92}}
 {id=Modulator_5, children [Random_5], properties {pitchMax=1.07, pitchMin=0.92}}
 {id=Modulator_6, children [Random_11], properties {pitchMax=1.5, pitchMin=1.2}}
 {id=Modulator_7, children [Random_6], properties {pitchMax=1.62, pitchMin=1.5}}
 {id=QualityLevel_0, children [Mixer_3,Mixer_0,Modulator_3,Modulator_4]}
 {id=Random_0, children [WavePlayer_0,WavePlayer_1], properties {weights [1,1]}}
 {id=Random_1, children [WavePlayer_3,WavePlayer_4], properties {weights [1,1]}}
 {id=Random_11, children [WavePlayer_41, WavePlayer_42, WavePlayer_43], properties {weights [1,1,1]}}
 {id=Random_2, children [WavePlayer_6,WavePlayer_7], properties {weights [1,1]}}
 {id=Random_5, children [WavePlayer_20, WavePlayer_21, WavePlayer_22], properties {weights [1,1,1]}}
 {id=Random_6, children [WavePlayer_23, WavePlayer_24, WavePlayer_55], properties {weights [1,1,1]}}
 {id=WavePlayer_0, properties {soundWaveAssetPtr=/Figment_S10/Sounds/Fort_Impact_Sounds/Demolition/Metal_Small_Impact06}}
 {id=WavePlayer_1, properties {soundWaveAssetPtr=/Figment_S10/Sounds/Fort_Impact_Sounds/Demolition/Metal_Small_Impact04}}
 {id=WavePlayer_10, properties {soundWaveAssetPtr=/Figment_S10/Sounds/Fort_Impact_Sounds/Demolition/Metal_Small_Impact_opt}}
 {id=WavePlayer_20, properties {soundWaveAssetPtr=&Metal_Small_Impact06}}
 {id=WavePlayer_21, properties {soundWaveAssetPtr=&Metal_Small_Impact04}}
 {id=WavePlayer_22, properties {soundWaveAssetPtr=/Figment_S10/Sounds/Fort_Impact_Sounds/Demolition/Metal_Small_Impact02}}
 {id=WavePlayer_23, properties {soundWaveAssetPtr=/Figment_S10/Sounds/Fort_Impact_Sounds/Demolition/Metal_Pole_Impact01}}
 {id=WavePlayer_24, properties {soundWaveAssetPtr=/Figment_S10/Sounds/Fort_Impact_Sounds/Demolition/Metal_Pole_Impact05}}
 {id=WavePlayer_3, properties {soundWaveAssetPtr=/Figment_S10/Sounds/Fort_Impact_Sounds/Demolition/Fort_Impact_Wobble_01}}
 {id=WavePlayer_4, properties {soundWaveAssetPtr=/Figment_S10/Sounds/Fort_Impact_Sounds/Demolition/Fort_Impact_Wobble_02}}
 {id=WavePlayer_41, properties {soundWaveAssetPtr=&Fort_Impact_Wobble_01}}
 {id=WavePlayer_42, properties {soundWaveAssetPtr=&Fort_Impact_Wobble_02}}
 {id=WavePlayer_43, properties {soundWaveAssetPtr=/Figment_S10/Sounds/Fort_Impact_Sounds/Demolition/Fort_Impact_Wobble_03}}
 {id=WavePlayer_55, properties {soundWaveAssetPtr=/Figment_S10/Sounds/Fort_Impact_Sounds/Demolition/Metal_Pole_Impact06}}
 {id=WavePlayer_6, properties {soundWaveAssetPtr=&Metal_Pole_Impact01}}
 {id=WavePlayer_7, properties {soundWaveAssetPtr=&Metal_Pole_Impact05}}
 {id=WavePlayer_9, properties {soundWaveAssetPtr=&Metal_Small_Impact_opt}}
]
properties {
 attenuationSettings=/Figment_S10/Sounds/Attenuation/Fort_Demolition_Close_Att
 maxDistance=2000
 soundClassObject=/Game/Sounds/FortSoundClassesAndModes/Fort_ImpactSounds_DuckDefault_PSM
 volumeMultiplier=0.5
 concurrencyOverrides {maxCount=3, resolutionRule=StopOldest}
 concurrencySet [&CONC_Object_Impacts_FourMax]
}
